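# Sound cue for basketball hit impacts.
# A random node picks one of three equally weighted hit samples, which passes through a modulator at default settings for slight variation.
# It uses a ball-bounce attenuation asset for spatial falloff and is routed to the emote music sound class rather than an impact effects class.
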SoundCue /Game/Sounds/Fort_Impact_Sounds/Basketball/BasketBall_Hit_Cue
root=Modulator_0
nodes [
 {id=Modulator_0, children [Random_1]}
 {id=Random_1, children [WavePlayer_0,WavePlayer_4,WavePlayer_5], properties {weights [1,1,1]}}
 {id=WavePlayer_0, properties {soundWaveAssetPtr=&/BasketBall_Hit_03}}
 {id=WavePlayer_4, properties {soundWaveAssetPtr=&/BasketBall_Hit_01}}
 {id=WavePlayer_5, properties {soundWaveAssetPtr=&/BasketBall_Hit_02}}
]
properties {
 attenuationSettings=/Game/Sounds/Attenuation/Fort_Ball_Bounce_Athena_Att
 maxDistance=7000
 soundClassObject=&Fort_Music_Emote
 volumeMultiplier=1
}
